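# Translucent two-sided trail material for the Silencer pickaxe's Niagara ribbons and meshes, which reacts to a weak-point hit count.
# It layers panned motion lines, soft and hard leading edges and a hard edge line under UV fades, with optional color-curve, two-color, HDR and multicolor emissive.
# The dynamic-parameter hit count, clamped and normalized, steps against a V gradient to mask a Reacted Color tint and reduce opacity there.
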Material /BRCosmetics/Weapons/FORT_Melee/Pickaxe_Silencer/FX/M_Melee_Trail_SilencerReactive
graph {
 properties {
  bUsedWithNiagaraMeshParticles=true
  bUsedWithNiagaraRibbons=true
  blendMode=BLEND_Translucent
  pixelDepthOffsetMode=PDOM_Legacy
  twoSided=true
  usedShadingModels=Default Lit
 }
 nodes [
  n1 = DynamicParameter() {defaultValue=<rgba 3 0 0 0>, paramNames [WeakPointHit Int, Waveform, 2, 3]}
  n2 = OneMinus(None=Divide(A=Clamp(None=@n1.WeakPointHit Int) {maxDefault=5}) {constB=5})
  n3 = Power(Base=ComponentMask(None=TextureCoordinate()) {g=true}) {constExponent=1.25}
  n4 = LinearInterpolate(
   B=VectorParameter() {parameterName=Reacted Color, defaultValue=<rgba 0.9618 0.0334 0.0334 0>}.RGB
   Alpha=NamedRerouteDeclaration(None=Step(Y=@n2, X=@n3)) {name=Lerp Alpha, nodeColor=<rgba 0 1 0.6719 1>}
  )
  n5 = VectorParameter() {group=Two Color Blend, parameterName=Two Color Blend A}
  n6 = VectorParameter() {group=Two Color Blend, parameterName=Two Color Blend B, sortPriority=1}
  n7 = TextureCoordinate()
  n8 = ComponentMask(None=@n7) {r=true}
  n9 = LinearInterpolate(
   A=@n8
   B=OneMinus(None=@n8)
   Alpha=Saturate(None=ScalarParameter() {group=Color Curve, parameterName=ColorGradient_HorizontalLeftORRight, sortPriority=2})
  )
  n10 = ComponentMask(None=@n7) {g=true}
  n11 = LinearInterpolate(
   A=@n10
   B=OneMinus(None=@n10)
   Alpha=Saturate(None=ScalarParameter() {group=Color Curve, parameterName=ColorGradient_VerticallUpOrDown, sortPriority=3})
  )
  n12 = ScalarParameter() {group=Color Curve, parameterName=SwizzleGradientDirection, sortPriority=1}
  n13 = Constant() {r=0.01}
  n14 = ComponentMask(None=TextureCoordinate()) {g=true}
  n15 = ScalarParameter() {defaultValue=0.02, group=Hard Line, parameterName=Edge Radius A, sortPriority=1}
  n16 = ScalarParameter() {defaultValue=0.005, group=Hard Line, parameterName=Edge Radius B, sortPriority=2}
  n17 = ComponentMask(None=TextureCoordinate()) {r=true}
  n18 = ScalarParameter() {defaultValue=0.3, group=Hard Line, parameterName=Edge Hardness A, sortPriority=3}
  n19 = ScalarParameter() {group=Hard Line, parameterName=Edge Hardness B, sortPriority=4}
  n20 = DynamicParameter() {defaultValue=<rgba 3 0 0 0>, paramNames [WeakPointHit Int, Waveform, 2, 3]}
  n21 = VectorParameter() {group=Soft Leading Edge, parameterName=Soft Leading Edge UV Scale, sortPriority=6, defaultValue=<rgba 0.2 1 0 0>}
  n22 = LinearInterpolate(
   A=TextureCoordinate()
   B=TextureCoordinate() {coordinateIndex=1}
   Alpha=ScalarParameter() {group=UV, parameterName=Use UV1 for Motion Lines A and Leading Edges}
  )
  n23 = ComponentMask(None=@n22) {r=true}
  n24 = LinearInterpolate(A=@n23, B=OneMinus(None=@n23), Alpha=ScalarParameter() {defaultValue=1, group=UV, parameterName=U Direction for Stretch, sortPriority=3})
  n25 = ScalarParameter() {defaultValue=2, group=UV, parameterName=UV Stretch Amount, sortPriority=2}
  n26 = AppendVector(
   A=LinearInterpolate(A=@n23, B=Power(Base=@n24, Exp=@n25), Alpha=ScalarParameter() {group=UV, parameterName=Stretch Trailing Edge, sortPriority=1})
   B=ComponentMask(None=@n22) {g=true}
  )
  n27 = VectorParameter() {group=Soft Leading Edge, parameterName=Soft Leading Edge Panner Speed, sortPriority=7, defaultValue=<rgba -0.65 0 0 0>}
  n28 = Panner(Coordinate=Multiply(A=@n26, B=ComponentMask(None=@n21.RGB) {g=true, r=true}), Speed=ComponentMask(None=@n27.RGB) {g=true, r=true}) {bFractionalPart=true}
  n29 = Add(A=@n28, B=AppendVector(A=@n20.WeakPointHit Int, B=@n20.WeakPointHit Int))
  n30 = ScalarParameter() {group=Soft Leading Edge, parameterName=Use_Dynamic_UV-Offsets, sortPriority=0}
  n31 = TextureSampleParameter2D(UVs=Add(A=NamedRerouteUsage(), B=LinearInterpolate(A=@n28, B=@n29, Alpha=@n30))) {
   group=Soft Leading Edge
   parameterName=Soft Leading Edge Texture
   samplerSource=SSM_Wrap_WorldGroupSettings
   sortPriority=4
   texture=&T_NoisePack_01
  }
  n32 = ChannelMaskParameter(None=@n31.RGBA) {group=Soft Leading Edge, parameterName=Soft Leading Edge Texture Mask, sortPriority=5}
  n33 = ScalarParameter() {defaultValue=-0.5, group=Soft Leading Edge, parameterName=Bias, sortPriority=3}
  n34 = TextureCoordinate()
  n35 = Add(
   A=Multiply(A=Add(A=@n32.R, B=@n33), B=ScalarParameter() {defaultValue=0.2, group=Soft Leading Edge, parameterName=Scale, sortPriority=2})
   B=OneMinus(None=ComponentMask(None=@n34) {r=true})
  )
  n36 = Multiply(
   A=SphereMask(
    A=AppendVector(A=@n13, B=@n14)
    B=@n13
    Radius=LinearInterpolate(A=@n15, B=@n16, Alpha=@n17)
    Hardness=LinearInterpolate(A=@n18, B=@n19, Alpha=@n17)
   )
   B=@n35
  )
  n37 = CurveAtlasRowParameter(CurveTime=LinearInterpolate(
    A=LinearInterpolate(A=@n9, B=@n11, Alpha=@n12)
    B=@n36
    Alpha=ScalarParameter() {group=Color Curve, parameterName=Hard Line Controls Color Curve, sortPriority=4}
   )) {
   atlas=&AthenaCurveAtlas01
   curve=/Game/Characters/Player/Male/Medium/Bodies/M_MED_BattleSuit/FX/BattleSuit_Purple
   defaultValue=87
   group=Color Curve
   parameterName=ColorCurve
  }
  n38 = VectorParameter() {group=Motion Lines, parameterName=Fine Motion Lines UV Scale A, sortPriority=5, defaultValue=<rgba 1 3.464 0 0>}
  n39 = VectorParameter() {group=Motion Lines, parameterName=Fine Motion Lines Panner Speed A, sortPriority=4, defaultValue=<rgba 0 0.976 0 0>}
  n40 = Panner(Coordinate=Multiply(A=@n26, B=ComponentMask(None=@n38.RGB) {g=true, r=true}), Speed=ComponentMask(None=@n39.RGB) {g=true, r=true})
  n41 = MaterialFunctionCall(XY=@n40) {materialFunction=&Swizzle}
  n42 = ScalarParameter() {defaultValue=1, group=Motion Lines, parameterName=Motion Lines Swizzle UVs, sortPriority=12}
  n43 = NamedRerouteUsage()
  n44 = TextureSampleParameter2D(UVs=Add(A=LinearInterpolate(A=@n40, B=@n41.YX, Alpha=@n42), B=@n43)) {
   group=Motion Lines
   parameterName=Motion Lines Texture
   samplerSource=SSM_Wrap_WorldGroupSettings
   texture=/Game/Effects/Fort_Effects/Textures/Outlander/T_InTheZone_LightRays_01
  }
  n45 = ChannelMaskParameter(None=@n44.RGBA) {group=Motion Lines, parameterName=Motion Lines Texture Mask A, sortPriority=1}
  n46 = ScalarParameter() {defaultValue=1.5, group=Motion Lines, parameterName=Motion Lines Texture A Contrast, sortPriority=2}
  n47 = VectorParameter() {group=Motion Lines, parameterName=Fine Motion Lines UV Scale B, sortPriority=11, defaultValue=<rgba 1 4.4362 0 0>}
  n48 = VectorParameter() {group=Motion Lines, parameterName=Fine Motion Lines Panner Speed B, sortPriority=10, defaultValue=<rgba 0 -0.7616 0 0>}
  n49 = Panner(Coordinate=Multiply(A=@n26, B=ComponentMask(None=@n47.RGB) {g=true, r=true}), Speed=ComponentMask(None=@n48.RGB) {g=true, r=true}) {speedY=-0.7616}
  n50 = MaterialFunctionCall(XY=@n49) {materialFunction=&Swizzle}
  n51 = TextureSampleParameter2D(UVs=Add(A=LinearInterpolate(A=@n49, B=@n50.YX, Alpha=@n42), B=@n43)) {
   group=Motion Lines
   parameterName=Motion Lines Texture
   samplerSource=SSM_Wrap_WorldGroupSettings
   sortPriority=6
   texture=&T_InTheZone_LightRays_01
  }
  n52 = ChannelMaskParameter(None=@n51.RGBA) {group=Motion Lines, maskChannel=Green, parameterName=Motion Lines Texture Mask B, sortPriority=7, defaultValue=<rgba 0 1 0 0>}
  n53 = ScalarParameter() {defaultValue=2, group=Motion Lines, parameterName=Motion Lines Texture B Contrast, sortPriority=8}
  n54 = Add(
   A=Multiply(
    A=ScalarParameter() {defaultValue=1, group=Motion Lines, parameterName=Motion Lines A Intensity, sortPriority=3}
    B=Power(Base=@n45.R, Exp=@n46) {constExponent=1.5}
   )
   B=Multiply(
    A=Power(Base=@n52.R, Exp=@n53)
    B=ScalarParameter() {defaultValue=1, group=Motion Lines, parameterName=Motion Lines B Intensity, sortPriority=9}
   )
  )
  n55 = ScalarParameter() {defaultValue=0.5, group=Motion Lines, parameterName=Combined Motion Lines Scale}
  n56 = Power(Base=@n35, Exp=ScalarParameter() {defaultValue=6, group=Soft Leading Edge, parameterName=Soft Leading Edge Contrast, sortPriority=1}) {constExponent=6}
  n57 = ScalarParameter() {defaultValue=1, group=Soft Leading Edge, parameterName=Use Soft Leading Edge}
  n58 = VectorParameter() {group=Hard Leading Edge, parameterName=Hard Leading Edge UV Scale A, sortPriority=4, defaultValue=<rgba 1.2 1.5 0 0>}
  n59 = MaterialFunctionCall(Float2=Multiply(A=@n26, B=ComponentMask(None=@n58.RGB) {g=true, r=true})) {materialFunction=&BreakOutFloat2Components}
  n60 = MaterialFunctionCall(
   X=OneMinus(None=@n59.G)
   Y=Add(A=@n59.R, B=ScalarParameter() {defaultValue=-0.3, group=Hard Leading Edge, parameterName=Hard Leading Edge Bias A, sortPriority=5}) {constA=0.25, constB=-0.3}
  ) {materialFunction=&MakeFloat2}
  n61 = VectorParameter() {group=Hard Leading Edge, parameterName=Hard Leading Edge UV Scale B, sortPriority=5, defaultValue=<rgba 12 3 0 0>}
  n62 = MaterialFunctionCall(Float2=Multiply(A=@n26, B=ComponentMask(None=@n61.RGB) {g=true, r=true})) {materialFunction=&BreakOutFloat2Components}
  n63 = MaterialFunctionCall(
   X=OneMinus(None=@n62.G)
   Y=Add(A=ScalarParameter() {defaultValue=0.25, group=Hard Leading Edge, parameterName=Hard Leading Edge Bias B, sortPriority=6}, B=@n62.R) {constA=0.25}
  ) {materialFunction=&MakeFloat2}
  n64 = ComponentMask(None=TextureCoordinate()) {g=true}
  n65 = VectorParameter() {group=Hard Leading Edge, parameterName=Hard Edge Mask Panner Speed, sortPriority=7}
  n66 = TextureSampleParameter2D(UVs=Panner(Coordinate=LinearInterpolate(A=@n60.Result, B=@n63.Result, Alpha=@n64), Speed=ComponentMask(None=@n65.RGB) {g=true, r=true})) {
   group=Hard Leading Edge
   parameterName=Hard Leading Edge Texture
   samplerType=SAMPLERTYPE_LinearColor
   sortPriority=2
   texture=/Game/CharClasses/Soldier/Traits/AdvancedTactics/Abilities/GoingCommando/Textures/T_GoingCommando_RingNoise_01
  }
  n67 = OneMinus(None=@n64)
  n68 = Multiply(
   A=ChannelMaskParameter(None=@n66.RGBA) {group=Hard Leading Edge, parameterName=Hard Leading Edge Texture Mask, sortPriority=3}.R
   B=Multiply(A=@n67, B=@n67)
  )
  n69 = ScalarParameter() {defaultValue=1, group=Hard Leading Edge, parameterName=Use Hard Leading Edge}
  n70 = LinearInterpolate(
   A=LinearInterpolate(
    A=ScalarParameter() {defaultValue=0.2, group=Alpha, parameterName=Fine Motion Lines Opacity A, sortPriority=10}
    B=ScalarParameter() {defaultValue=0.3, group=Alpha, parameterName=Fine Motion Lines Opacity B, sortPriority=11}
    Alpha=Multiply(A=@n54, B=@n55)
   ) {constA=1}
   B=ScalarParameter() {defaultValue=0.8, group=Alpha, parameterName=Leading Edge Opacity, sortPriority=9}
   Alpha=Add(A=Multiply(A=@n56, B=@n57), B=Multiply(A=@n68, B=@n69)) {constB=0}
  ) {constAlpha=0}
  n71 = ComponentMask(None=TextureCoordinate()) {g=true}
  n72 = Multiply(
   A=Add(A=@n71, B=ScalarParameter() {group="Fade Top && Bottom Length Edges", parameterName=Top Length Fade Bias, sortPriority=2})
   B=ScalarParameter() {defaultValue=5, group="Fade Top && Bottom Length Edges", parameterName=Top Length Fade Scale}
  ) {constB=3}
  n73 = Multiply(
   A=Add(A=OneMinus(None=@n71), B=ScalarParameter() {group="Fade Top && Bottom Length Edges", parameterName=Bottom Length Fade Bias, sortPriority=3})
   B=ScalarParameter() {defaultValue=5, group="Fade Top && Bottom Length Edges", parameterName=Bottom Length Fade Scale, sortPriority=1}
  ) {constB=2}
  n74 = Multiply(
   A=LinearInterpolate(B=Saturate(None=@n72), Alpha=ScalarParameter() {defaultValue=1, group=Alpha, parameterName=Fade Top Length Edge, sortPriority=7}) {constA=1}
   B=LinearInterpolate(B=Saturate(None=@n73), Alpha=ScalarParameter() {defaultValue=1, group=Alpha, parameterName=Fade Bottom Length Edge, sortPriority=8}) {constA=1}
  )
  n75 = VectorParameter() {group=Trailing Edge Sphere Mask, parameterName=Trailing Edge Sphereical UV Scale, sortPriority=4, defaultValue=<rgba 0.5 1 0 0>}
  n76 = Multiply(A=TextureCoordinate(), B=ComponentMask(None=@n75.RGB) {g=true, r=true})
  n77 = VectorParameter() {group=Trailing Edge Sphere Mask, parameterName="Trailing Edge Sphere Mask Coord (B)", sortPriority=3, defaultValue=<rgba 0.85 0.82 0 0>}
  n78 = ComponentMask(None=@n77.RGB) {g=true, r=true}
  n79 = ScalarParameter() {defaultValue=0.8, group=Trailing Edge Sphere Mask, parameterName=Trailing Edge Radius, sortPriority=1}
  n80 = ScalarParameter() {defaultValue=0.3, group=Trailing Edge Sphere Mask, parameterName=Trailing Edge Hardness, sortPriority=2}
  n81 = Power(
   Base=OneMinus(None=SphereMask(A=@n76, B=@n78, Radius=@n79, Hardness=@n80))
   Exp=ScalarParameter() {defaultValue=5, group=Trailing Edge Sphere Mask, parameterName=Trailing Edge Contrast}
  ) {constExponent=5}
  n82 = ScalarParameter() {defaultValue=1, group=Alpha, parameterName=Multiply Trailing Edge Sphere Mask, sortPriority=6}
  n83 = Add(
   A=LinearInterpolate(B=@n36, Alpha=ScalarParameter() {defaultValue=1, group=Hard Line, parameterName=Add Top Edge Hard Line})
   B=Multiply(A=Multiply(A=@n70, B=@n74), B=LinearInterpolate(B=@n81, Alpha=@n82) {constA=1})
  )
  n84 = Add(A=@n83, B=LinearInterpolate(B=@n37, Alpha=ScalarParameter() {group=Emissive Color, parameterName=Multiply by Color Curve}))
  n85 = LinearInterpolate(
   A=LinearInterpolate(
    B=LinearInterpolate(A=@n5.RGB, B=@n6.RGB, Alpha=@n84)
    Alpha=ScalarParameter() {group=Emissive Color, parameterName="Multiply by Two Color Blend ", sortPriority=2}
   ) {constA=1}
   B=VectorParameter() {parameterName=Reacted Color, defaultValue=<rgba 0.9618 0.0334 0.0334 0>}.RGB
   Alpha=NamedRerouteUsage()
  )
  n86 = LinearInterpolate(B=@n37, Alpha=ScalarParameter() {group=Emissive Color, parameterName=Multiply by Color Curve}) {constA=1}
  n87 = VectorParameter() {group=Main Texture, parameterName=Main Texture UV Offset, sortPriority=4}
  n88 = VectorParameter() {group=Main Texture, parameterName=Main Texture UV Scale, sortPriority=3, defaultValue=<rgba 1 1 0 0>}
  n89 = Multiply(A=Add(A=@n7, B=ComponentMask(None=@n87.RGB) {g=true, r=true}), B=ComponentMask(None=@n88.RGB) {g=true, r=true})
  n90 = MaterialFunctionCall(XY=@n89) {materialFunction=&Swizzle}
  n91 = ScalarParameter() {group=Main Texture, parameterName=Main Texture Sizzle UVs, sortPriority=5}
  n92 = VectorParameter() {group=Main Texture, parameterName=Main Texture Panner Speed, sortPriority=2}
  n93 = TextureSampleParameter2D(UVs=Panner(Coordinate=LinearInterpolate(A=@n89, B=@n90.YX, Alpha=@n91), Speed=ComponentMask(None=@n92.RGB) {g=true, r=true})) {
   group=Main Texture
   parameterName=Texture
   samplerSource=SSM_Wrap_WorldGroupSettings
   texture=/Game/Effects/Fort_Effects/Textures/Grad/T_HolidayTreeSword_Gradient
  }
  n94 = ScalarParameter() {group=Emissive Color, parameterName="Multiply by Main TextureMask ", sortPriority=1}
  n95 = ParticleColor()
  n96 = Multiply(A=Multiply(A=Multiply(A=@n85, B=@n86), B=LinearInterpolate(B=@n93.RGB, Alpha=@n94) {constA=1}), B=@n95.RGB)
  n97 = Clamp(None=@n96)
  n98 = ScalarParameter() {defaultValue=100, group=Derive HDR from LDR, parameterName=HDR Dynamic Range, sortPriority=3}
  n99 = LinearInterpolate(
   B=LinearInterpolate(Alpha=@n36) {constA=0.01, constB=0.8}
   Alpha=ScalarParameter() {group=Derive HDR from LDR, parameterName=Mutilpy HDR Intensity by Hard Line, sortPriority=2}
  ) {constA=1}
  n100 = Multiply(A=ScalarParameter() {defaultValue=1, group=Derive HDR from LDR, parameterName=HDR Intensity, sortPriority=1}, B=@n99)
  n101 = ScalarParameter() {defaultValue=1, group=Derive HDR from LDR, parameterName=Mobile HDR Intensity, sortPriority=4}
  n102 = Multiply(A=@n96, B=@n99)
  n103 = LinearInterpolate(
   A=@n96
   B=ShadingPathSwitch(
    Default=MaterialFunctionCall(LDR Input=@n97, HDR Dynamic Range=@n98, HDR Tint=@n96, HDR Intensity Multiplier=@n100) {materialFunction=&DeriveHDRfromLDR}.RemappedRange
    Mobile=Multiply(A=@n101, B=@n102)
   ) {desc=Migrated from FeatureLevelSwitch}
   Alpha=ScalarParameter() {group=Derive HDR from LDR, parameterName=Use HDR Emissive Color}
  )
  n104 = ScalarParameter() {defaultValue=1, group=Multicolor, parameterName=MultiColor_Brightness, sortPriority=1}
  n105 = ScalarParameter() {defaultValue=1, group=Multicolor, parameterName=MultiColor_Exponent, sortPriority=2}
  n106 = VertexNormalWS()
  n107 = TextureObjectParameter() {
   group=Multicolor
   parameterName=MultiColor_GradientTex
   sortPriority=4
   texture=/Game/Characters/Player/Female/Medium/Bodies/F_Med_Soldier_01/Skins/F_MED_COM_Taxi_Upgraded/Skins/Multi/T_multicolor_gradient_v2_D
  }
  n108 = MaterialFunctionCall!(
   Phase Shift=ScalarParameter() {defaultValue=1, group=Multicolor, parameterName=MultiColor_GradientShift, sortPriority=3}
   UV=Clamp(None=MaterialFunctionCall(Brightness=@n104, Exponent=@n105, Vertex / Pixel Normal=@n106) {materialFunction=&MF_Fresnel_FX}.Fresnel Double Sided)
   height texture=Constant()
   Gradient=@n107
   Phases=ScalarParameter() {defaultValue=4, group=Multicolor, parameterName=MultiColor_GradientPhases, sortPriority=5}
  ) {materialFunction=/Game/Characters/Player/Female/Medium/Bodies/F_Med_Soldier_01/Skins/F_MED_COM_Taxi_Upgraded/Skins/Multi/MF_MulticolorGrad}
  n109 = NamedRerouteUsage()
  n110 = LinearInterpolate(
   A=@n103
   B=LinearInterpolate(A=@n108.Result, B=Add(A=@n109, B=@n108.Result), Alpha=ScalarParameter() {group=Multicolor, parameterName=MultiColor_Add, sortPriority=6})
   Alpha=ScalarParameter() {group=Multicolor, parameterName=Use_MultiColor, sortPriority=0}
  )
  MaterialFunctionCall(Emissive=Multiply(A=NamedRerouteUsage(), B=Add(A=@n4, B=@n110))) {materialFunction=&MF_FX_EmissiveAdjustments}
  n111 = LinearInterpolate(B=@n37.A, Alpha=ScalarParameter() {group=Alpha, parameterName=Multiply by Color Curve Alpha}) {constA=1}
  n112 = ChannelMaskParameter(None=@n93.RGBA) {group=Main Texture, maskChannel=Alpha, parameterName=Main Texture Alpha Mask, sortPriority=1, defaultValue=<rgba 0 0 0 1>}
  n113 = LinearInterpolate(B=@n112.R, Alpha=ScalarParameter() {group=Alpha, parameterName=Multiply by Main Texture Alpha Mask, sortPriority=1}) {constA=1}
  n114 = ComponentMask(None=TextureCoordinate()) {r=true}
  n115 = Multiply(
   A=Add(A=@n114, B=ScalarParameter() {group="Fade Leading && Trailing Edges", parameterName=Leading Edge Fade Bias, sortPriority=2})
   B=ScalarParameter() {defaultValue=3, group="Fade Leading && Trailing Edges", parameterName=Leading Edge Fade Scale}
  ) {constB=3}
  n116 = Multiply(
   A=Add(A=OneMinus(None=@n114), B=ScalarParameter() {group="Fade Leading && Trailing Edges", parameterName=Trailing Edge Fade Bias, sortPriority=3})
   B=ScalarParameter() {defaultValue=2, group="Fade Leading && Trailing Edges", parameterName=Trailing Edge Fade Scale, sortPriority=1}
  ) {constB=2}
  n117 = Multiply(
   A=LinearInterpolate(B=Saturate(None=@n115), Alpha=ScalarParameter() {defaultValue=1, group=Alpha, parameterName=Fade Leading Edge, sortPriority=4}) {constA=1}
   B=LinearInterpolate(B=Saturate(None=@n116), Alpha=ScalarParameter() {defaultValue=1, group=Alpha, parameterName=Fade Trailing Edge, sortPriority=5}) {constA=1}
  )
  n118 = ScalarParameter() {defaultValue=64, group=Alpha, parameterName=NearFadeDistance, sortPriority=3}
  n119 = Multiply(
   A=Multiply(A=Multiply(A=@n111, B=@n113), B=@n95.A)
   B=Multiply(
    A=Multiply(A=@n83, B=@n117)
    B=MaterialFunctionCall(Distance=@n118) {materialFunction=&Near_Camera_Fade}.Result
   )
  )
  n120 = ShadingPathSwitch(
   Default=DepthFade(Opacity=@n119, FadeDistance=ScalarParameter() {defaultValue=50, group=Alpha, parameterName=FadeDistance, sortPriority=2})
   Mobile=@n119
  ) {desc=Migrated from FeatureLevelSwitch}
  n121 = MaterialFunctionCall(
   Brightness=ScalarParameter() {defaultValue=1, group=Alpha, parameterName="CameraNormal_Fresnel_Brightness (S)"}
   Exponent=ScalarParameter() {defaultValue=3, group=Alpha, parameterName="CameraNormal_Fresnel_Exponent (S)"}
   Vertex / Pixel Normal=@n106
  ) {materialFunction=&MF_Fresnel_FX}
  n122 = LinearInterpolate(
   A=@n120
   B=Multiply(A=@n120, B=Multiply(A=@n121.Fresnel Double Sided))
   Alpha=ScalarParameter() {group=Alpha, parameterName=Use_CameraNormal_Fresnel}
  )
  n123 = Multiply(A=NamedRerouteUsage(), B=ScalarParameter() {defaultValue=0.5, parameterName=Rect Color Alpha Scale})
  Multiply(A=@n122, B=OneMinus(None=@n123))
  n124 = Multiply(A=VertexColor().R, B=ScalarParameter() {defaultValue=0.5, group=Vertex Color, parameterName=Vertex Color Offset, sortPriority=1})
  n125 = Constant()
  NamedRerouteDeclaration(None=StaticSwitchParameter(True=AppendVector(A=@n124, B=@n125), False=Constant()) {group=Vertex Color, parameterName=VertexColorOffset, sortPriority=0}) {name=VertexColorOffset, nodeColor=<rgba 0 1 0.8125 1>}
  NamedRerouteDeclaration(None=@n103) {name=Emissive, nodeColor=<rgba 0.3828 0 1 1>}
  NamedRerouteDeclaration(None=MaterialFunctionCall(
    Color_Morning=VectorParameter() {group=Time of Day, parameterName=Color Morning, defaultValue=<rgba 0.2795 0.2795 0.2795 1>}.RGB
    Color_Day=VectorParameter() {group=Time of Day, parameterName=Color Day, defaultValue=<rgba 1 1 1 1>}.RGB
    Color_Evening=VectorParameter() {group=Time of Day, parameterName=Color Evening, defaultValue=<rgba 0.2726 0.2726 0.2726 1>}.RGB
    Color_Night=VectorParameter() {group=Time of Day, parameterName=Color Night, defaultValue=<rgba 0.2576 0.2951 0.2951 1>}.RGB
   ) {materialFunction=&Color_Time_of_Day}.Result) {name=Time of Day Color, nodeColor=<rgba 1 0.2578 0 1>}
  Subtract(A=@n122, B=@n123)
 ]
}
