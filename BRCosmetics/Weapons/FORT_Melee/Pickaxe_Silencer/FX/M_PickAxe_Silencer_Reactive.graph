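# Lit reactive material for the Silencer pickaxe, usable on Niagara mesh particles, that shifts from base color toward Fresnel-weighted reacted colors.
# The WeakPointHit dynamic parameter, clamped and normalized over five steps, drives a mask by threshold or optional UV pan.
# Mask-driven lerps blend color, two normal maps and metal/spec/roughness sets.
# Emissive follows the Waveform parameter, plus switchable fake specular from normals dotted with a time-animated vector.
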Material /BRCosmetics/Weapons/FORT_Melee/Pickaxe_Silencer/FX/M_PickAxe_Silencer_Reactive
graph {
 properties {bUsedWithNiagaraMeshParticles=true, pixelDepthOffsetMode=PDOM_Legacy, usedShadingModels=Default Lit}
 nodes [
  n1 = NamedRerouteUsage()
  n2 = VectorParameter() {parameterName=Reacted ColorA, defaultValue=<rgba 0.9618 0.0334 0.0334 0>}
  n3 = TextureCoordinate()
  n4 = VectorParameter() {parameterName=UV Scale, channelNames {a=V2, b=U2, g=V1, r=U1}, defaultValue=<rgba 2 2 4 4>}
  n5 = Multiply(A=Rotator(Coordinate=@n3, Time=ScalarParameter() {parameterName=Rotate UV 2}), B=AppendVector(A=@n4.U2, B=@n4.U2))
  n6 = TextureSampleParameter2D(UVs=@n5) {parameterName=Texture, texture=/Game/Accessories/FORT_Backpacks/M_MED_Heist/Textures/T_FN_Character_Heist_M}
  n7 = VectorParameter() {parameterName=Reacted ColorB, defaultValue=<rgba 0.7396 0.0307 0 0>}
  n8 = Divide(A=Clamp(None=DynamicParameter() {defaultValue=<rgba 3 0 0 0>, paramNames [WeakPointHit Int, Waveform, 2, 3]}.WeakPointHit Int) {maxDefault=5}) {constB=5}
  n9 = AppendVector(A=Constant(), B=@n8)
  n10 = TextureObjectParameter() {parameterName=FX Mask Texture, texture=/BRCosmetics/Weapons/FORT_Melee/Pickaxe_SharpFang/Textures/T_SharpFang_FX}
  n11 = TextureSample(UVs=Add(A=TextureCoordinate(), B=Multiply(A=@n9) {constB=0.5}), Tex=@n10) {samplerType=SAMPLERTYPE_LinearColor, texture=&/T_PickAxe_Sword_FX_Mask}
  n12 = TextureSample(Tex=@n10) {samplerType=SAMPLERTYPE_LinearColor, texture=&T_PickAxe_Sword_FX_Mask}
  n13 = OneMinus(None=@n8)
  n14 = Saturate(None=Multiply(A=Subtract(A=@n12.R, B=@n13), B=ScalarParameter() {defaultValue=1, parameterName=Mask Strength Adjustment}) {constB=10})
  n15 = LinearInterpolate(
   A=VectorParameter() {parameterName=Base Color}.RGB
   B=Multiply(A=Fresnel() {baseReflectFraction=0.05, exponent=11}, B=LinearInterpolate(A=@n2.RGB, B=@n7.RGB, Alpha=@n6.R))
   Alpha=NamedRerouteDeclaration(None=StaticSwitchParameter(True=@n11.RGB, False=@n14) {parameterName=Use UV Pan Based Mask}) {name=Lerp Alpha, nodeColor=<rgba 1 0.6797 0 1>}
  )
  n16 = ParticleColor()
  n17 = NamedRerouteUsage()
  Add(
   A=Multiply(A=@n1, B=Multiply(A=@n2.RGB, B=@n6.B))
   B=Multiply(A=Multiply(A=@n15, B=@n16.RGB), B=LinearInterpolate(B=@n6.R, Alpha=@n17) {constA=1})
  )
  n18 = ScalarParameter() {defaultValue=1, parameterName=Nomal Strength 1}
  n19 = ScalarParameter() {parameterName=Rotate UV 1}
  n20 = TextureSampleParameter2D(UVs=Multiply(A=Rotator(Coordinate=@n3, Time=@n19), B=ComponentMask(None=@n4) {g=true, r=true})) {parameterName=Normal Map 1, samplerType=SAMPLERTYPE_Normal, texture=/Game/Effects/Fort_Effects/Textures/Misc/T_NoiseNormal}
  n21 = OneMinus(None=ComponentMask(None=@n20.RGB) {g=true, r=true})
  n22 = TextureSampleParameter2D(UVs=@n5) {
   parameterName=Normal Map 2
   samplerType=SAMPLERTYPE_Normal
   texture=/Game/Accessories/FORT_Backpacks/M_MED_Heist/Demo/Textures/T_M_MED_Heist_Backpack_N
  }
  n23 = ComponentMask(None=@n22.RGB) {g=true, r=true}
  n24 = ScalarParameter() {defaultValue=1, parameterName=NOrmal 2 Pow}
  n25 = AppendVector(A=Multiply(A=ScalarParameter() {defaultValue=2, parameterName=Nomal Strength 2}, B=Power(Base=@n23, Exp=@n24)), B=@n22.B)
  LinearInterpolate(A=AppendVector(A=Multiply(A=@n18, B=@n21), B=@n20.B), B=@n25, Alpha=NamedRerouteUsage())
  n26 = Multiply(A=@n2.RGB, B=DynamicParameter() {defaultValue=<rgba 3 0 0 0>, paramNames [WeakPointHit Int, Waveform, 2, 3]}.Waveform)
  n27 = ScalarParameter() {defaultValue=0.14, parameterName=Emissive Scale}
  n28 = Add(
   A=Multiply(A=Multiply(A=@n2.RGB, B=@n1), B=ScalarParameter() {defaultValue=0.05, parameterName=Default Emissive Scale})
   B=Multiply(A=Multiply(A=@n26, B=@n27), B=@n6.R)
  )
  n29 = VectorParameter() {parameterName=Base Specular Masks Scale, channelNames {a="", b="", g=V, r=U}, defaultValue=<rgba 1 1 0 0>}
  n30 = TextureSampleParameter2D(UVs=Multiply(A=TextureCoordinate(), B=ComponentMask(None=@n29) {g=true, r=true})) {parameterName=Base Specular Masks, texture=&/T_Pickaxe_Silencer_FX_S}
  n31 = Time()
  n32 = Sine(None=Multiply(A=@n31, B=ScalarParameter() {defaultValue=0.175, parameterName=Fake Specular Anim1}) {constB=0.31})
  n33 = Cosine(None=Multiply(A=@n31, B=ScalarParameter() {defaultValue=0.097, parameterName=Fake Specular Anim2}) {constB=0.1})
  n34 = DotProduct(A=Add(A=Multiply(A=@n25) {constB=2}) {constB=-1}, B=AppendVector(A=AppendVector(A=@n32, B=@n33), B=Constant()))
  n35 = Multiply(A=Multiply(A=@n34, B=@n34), B=ScalarParameter() {defaultValue=0.01, parameterName=Fake Light Scale})
  n36 = Multiply(A=@n6.G, B=Multiply(A=Multiply(A=@n30.R, B=@n35), B=NamedRerouteUsage()))
  MaterialFunctionCall(Emissive=StaticSwitchParameter(True=Add(A=@n28, B=@n36), False=@n28) {defaultValue=true, parameterName=Use Fake Specular}) {materialFunction=&MF_FX_EmissiveAdjustments}
  n37 = VectorParameter() {parameterName=Params 1, channelNames {a="", b=Rogh, g=Spec, r=Metal}, defaultValue=<rgba 0.5 0.7 0.3 0>}
  n38 = AppendVector(A=ComponentMask(None=@n30.RGB) {g=true, r=true}, B=Constant() {r=1})
  n39 = VectorParameter() {parameterName=Params 2, channelNames {a="", b=Rogh, g=Spec, r=Metal}, defaultValue=<rgba 0.5 1 0.05 0>}
  MaterialFunctionCall(Float3=LinearInterpolate(A=Multiply(A=@n37, B=@n30.B), B=Multiply(A=@n38, B=@n39), Alpha=NamedRerouteUsage())) {materialFunction=&BreakOutFloat3Components}
  n40 = Time() {period=1}
  Multiply(A=Frac(None=Multiply(A=@n40) {constB=0.5})) {constB=5}
 ]
}
