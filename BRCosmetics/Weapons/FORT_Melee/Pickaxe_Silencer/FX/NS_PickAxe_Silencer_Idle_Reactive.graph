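# Idle effect for the Silencer pickaxe that reacts to the weak-point hit count, offset by a Variants parameter.
# Mesh_Sword passes that level and a waveform pulse gated at higher levels to its material, and drives a flickering red light whose height and radius follow it.
# Brink sprite and Brink001 mesh glows switch on through alpha once the level passes 4, and are brighter in the front end.
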NiagaraSystem /BRCosmetics/Weapons/FORT_Melee/Pickaxe_Silencer/FX/NS_PickAxe_Silencer_Idle_Reactive
dependencies {
 usedDataInterfaces []
 usedDynamicInputs [&MakeCustomIntFromBool, &SetBoolByFloatComparison]
 usedModules [
  &SystemState
  &FN_Integrator
  &self:SystemSpawnScript.SystemScriptSource.SystemScriptGraph.NiagaraNodeAssignment_1.SetVariables_#1
 ]
}
emitters {
 Brink {
  emitterUpdateScript [
   &EmitterState(Life Cycle Mode=::NewEnumerator1, Loop Behavior=::NewEnumerator1, Loop Duration=1, UseLoopDelay=false)
   &SpawnBurst_Instantaneous(Spawn Group=1, Use Spawn Probability=false, UseLoopCountLimit=false) {module=SpawnBurst_Instantaneous001}
  ]
  particleSpawnScript [
   &InitializeParticle(
    Color=<rgba 5 0.0781 0.0781 0.13>
    Position Offset=<0 0 -21>
    Sprite Size Mode=::NewEnumerator3
    Uniform Sprite Size=16
    UsePositionOffset=true
    Write Material Random=true
   )
   &CameraOffset()
  ]
  particleUpdateScript [
   &ParticleState(Kill Particles When Lifetime Has Elapsed=false)
   &ScaleColor(
    Scale Alpha=&Multiply_Float(
     A=1
     B=&ClampFloat(Float=&MakeFloatFromInt(Int=&MakeIntFromFloat(Float=&FN_FrontEndFloatModifier(
         Gameplay=&Subtract_Float(
          A=&MakeFloatFromInt(Int=&Add_Integer(
            A=$System.ClampHitWeakPoint
            B=&MakeIntFromFloat(Float=HlslExpression() {hlslExpression="clamp(User.Variants, -5,5)"})
           ))
          B=4
         )
         Front End=&Subtract_Float(
          A=&MakeFloatFromInt(Int=&Add_Integer(
            A=$System.ClampHitWeakPoint
            B=&MakeIntFromFloat(Float=HlslExpression() {hlslExpression="clamp(User.Variants, -5,5)"})
           ))
          B=4
         )
        ))))
    )
    Scale RGB=&Multiply_VectorByFloat(Vector=<1 1 1>, Float=&FN_FrontEndFloatModifier(Gameplay=0.2, Front End=0.9))
    ScaleA=true
    ScaleRGB=true
   )
  ]
  rendererClasses [&NiagaraSpriteRendererProperties]
  renderers [{rendererClass=&NiagaraSpriteRendererProperties}]
 }
 Brink001 {
  emitterUpdateScript [
   &EmitterState(Life Cycle Mode=::NewEnumerator1, Loop Behavior=::NewEnumerator1, Loop Duration=1, UseLoopDelay=false)
   &SpawnBurst_Instantaneous(Use Spawn Probability=false, UseLoopCountLimit=false)
  ]
  particleSpawnScript [
   &InitializeParticle(
    Color=<rgba 5 0.0781 0.0781 0.13>
    Mesh Scale Mode=::NewEnumerator3
    Position Offset=<0 0 85>
    Position Offset Coordinate Space=::Local
    UsePositionOffset=true
    Write Material Random=true
   )
   &CameraOffset(Camera Offset Amount=1)
  ]
  particleUpdateScript [
   &ParticleState(Kill Particles When Lifetime Has Elapsed=false)
   &ScaleColor(
    Scale Alpha=&Multiply_Float(
     A=1
     B=&ClampFloat(Float=&MakeFloatFromInt(Int=&MakeIntFromFloat(Float=&FN_FrontEndFloatModifier(
         Gameplay=&Subtract_Float(
          A=&MakeFloatFromInt(Int=&Add_Integer(
            A=$System.ClampHitWeakPoint
            B=&MakeIntFromFloat(Float=HlslExpression() {hlslExpression="clamp(User.Variants, -5,5)"})
           ))
          B=4
         )
         Front End=&Subtract_Float(
          A=&MakeFloatFromInt(Int=&Add_Integer(
            A=$System.ClampHitWeakPoint
            B=&MakeIntFromFloat(Float=HlslExpression() {hlslExpression="clamp(User.Variants, -5,5)"})
           ))
          B=4
         )
        ))))
    )
    Scale RGB=&Multiply_VectorByFloat(Vector=<1 1 1>, Float=&FN_FrontEndFloatModifier(Gameplay=0.2, Front End=0.6))
    ScaleA=true
    ScaleRGB=true
   )
  ]
  rendererClasses [&NiagaraMeshRendererProperties]
  renderers [
   {rendererClass=&NiagaraMeshRendererProperties, rendererIndex=0}
   {rendererClass=&NiagaraMeshRendererProperties, rendererIndex=1}
  ]
 }
 Mesh_Sword {
  emitterUpdateScript [
   &EmitterState(Life Cycle Mode=::NewEnumerator1, Loop Behavior=::NewEnumerator1, Loop Duration=1, UseLoopDelay=false)
   &SpawnBurst_Instantaneous(Use Spawn Probability=false, UseLoopCountLimit=false)
  ]
  particleSpawnScript [
   &InitializeParticle(
    Color Mode=::NewEnumerator0
    Mesh Scale Mode=::NewEnumerator3
    Sprite Size Mode=::NewEnumerator3
    Uniform Sprite Size=100
    UsePositionOffset=false
    Write Material Random=true
   )
   &CameraOffset(Camera Offset Amount=0.05)
  ]
  particleUpdateScript [
   &ParticleState(Kill Particles When Lifetime Has Elapsed=false)
   &SolveForcesAndVelocity(Clamp Velocity=false, Limit Acceleration=false) {enabled=false}
   &DynamicMaterialParameters(
    Index 0 Param 1=&MakeFloatFromInt(Int=&Add_Integer(
      A=&Add_Integer(
       A=$User.HitWeakPoint
       B=&MakeIntFromFloat(Float=&FN_FrontEndFloatModifier(Gameplay=0, Front End=0))
      )
      B=&MakeIntFromFloat(Float=$User.Variants)
     ))
    Index 0 Param 2=&Multiply_Float(
     A=&Waveform(
      Waveform Count=::NewEnumerator2
      "Waveform [1]"=::NewEnumerator0
      "[1] Frequency"=1
      "[1] Amplitude Min/Max"=<5 1>
      "[2] Average"=false
      "[2] Input Position"=0
      "[2] Phase"=$Emitter.Age
      "[2] Phase Scale"=1.23
      "[2] Compound Frequency"=2
      "[2] Frequency"=1.25
      "[2] Amplitude Min/Max"=<0.5 1>
      "[2] Scale"=0.5
      "[2] Bias"=0
      "[2] Exponent"=1
      "[2] Pulse Duration"=0.1
      "[2] Clip Triangle Min/Max"=<0 0>
      "[2] Random Seed"=0
      "Waveform [3] Blend"=::NewEnumerator2
      "[3] Input Position"=0
      "[3] Phase"=$Emitter.Age
      "[3] Phase Scale"=0.5
      "[3] Frequency"=0.9
      "[3] Compound Frequency"=2
      "[3] Amplitude Min/Max"=<0.5 1>
      "[3] Scale"=1
      "[3] Bias"=0
      "[3] Exponent"=1
      "[3] Pulse Duration"=0.1
      "[3] Clip Triangle Min/Max"=<0 0>
      "[3] Random Seed"=0
     )
     B=&ClampFloat(Float=&MakeFloatFromInt(Int=/Niagara/DynamicInputs/Subtract/Subtract_Integer(
        A:NiagaraInt32=&MakeIntFromFloat(Float=&MakeFloatFromInt(Int=&Add_Integer(A=$User.HitWeakPoint, B=&MakeIntFromFloat(Float=$User.Variants))))
        B:NiagaraInt32=4
       )))
    )
    Module.Param0WriteEnabled=true
    Module.Param1WriteEnabled=true
    Module.Param2WriteEnabled=true
    Module.Param3WriteEnabled=true
   )
   &ScaleColor(Scale RGB=&Multiply_VectorByFloat(Vector=<1 1 1>, Float=1), ScaleA=true, ScaleRGB=true) {enabled=false}
   &FN_Light_Attributes(
    Color=<rgba 0.6 0.0052 0 1>
    Create Color Attribute=true
    Create Position Attribute=true
    Create Radius Attribute=true
    Enable Particle Light Enable=true
    Light Enabled:NiagaraBool=&SetBoolByIntComparison(
     A=0
     B=&Add_Integer(A=$User.HitWeakPoint, B=&MakeIntFromFloat(Float=$User.Variants))
     Comparison Type=::NewEnumerator6
    )
    Position=&MakeVector(
     X=0
     Y=0
     Z=&Multiply_Float(A=50, B=HlslExpression() {hlslExpression="max(((clamp(User.HitWeakPoint,0,5) +1) /6.0) , clamp(User.Variants, 0,1))"})
    )
    Radius=&Multiply_Float(
     A=HlslExpression() {hlslExpression="110 * clamp(User.Variants+1, 0,1)"}
     B=HlslExpression() {hlslExpression="clamp(0, 1,(clamp(User.HitWeakPoint,0,5) +1) /6.0 +.3 + User.Variants)"}
    )
    Scale Color=&Multiply_Float(
     A=&Waveform(
      Waveform Count=::NewEnumerator2
      "Waveform [1]"=::NewEnumerator0
      "[1] Frequency"=1
      "[1] Amplitude Min/Max"=<0.85 1>
      "[2] Average"=false
      "[2] Input Position"=0
      "[2] Phase"=$Emitter.Age
      "[2] Phase Scale"=1
      "[2] Compound Frequency"=2
      "[2] Frequency"=2
      "[2] Amplitude Min/Max"=<0.5 0.37>
      "[2] Scale"=1
      "[2] Bias"=0
      "[2] Exponent"=1
      "[2] Pulse Duration"=0.1
      "[2] Clip Triangle Min/Max"=<0 0>
      "[2] Random Seed"=0
      "[3] Average"=false
      "[3] Input Position"=0
      "[3] Phase"=$Emitter.Age
      "[3] Phase Scale"=1
      "[3] Frequency"=4
      "[3] Compound Frequency"=2
      "[3] Amplitude Min/Max"=<0.69 0.7>
      "[3] Scale"=1
      "[3] Bias"=0
      "[3] Exponent"=1
      "[3] Pulse Duration"=0.1
      "[3] Clip Triangle Min/Max"=<0 0>
      "[3] Random Seed"=0
     )
     B=1
    )
   )
  ]
  rendererClasses [&NiagaraMeshRendererProperties, &NiagaraLightRendererProperties]
  renderers [
   {rendererClass=&NiagaraMeshRendererProperties, rendererIndex=0}
   {rendererClass=&NiagaraLightRendererProperties, rendererIndex=1}
  ]
 }
}
summary {
 systemName=NS_PickAxe_Silencer_Idle_Reactive
 emitters [
  {emitterName=Mesh_Sword, rendererClasses [&NiagaraMeshRendererProperties, &NiagaraLightRendererProperties]}
  {emitterName=Brink, rendererClasses [&NiagaraSpriteRendererProperties]}
  {emitterName=Brink001, rendererClasses [&NiagaraMeshRendererProperties]}
 ]
 userVariables [
  User.bIsFrontEnd {type=NiagaraBool}
  User.bIsFrontEndPreview {type=NiagaraBool}
  User.bIsFullyDeployed {type=NiagaraBool}
  User.ForwardDot {type=NiagaraFloat, defaultValue=0, description=""}
  User.HitWeakPoint {type=NiagaraInt32}
  User.PlayerSpeed {type=NiagaraFloat, defaultValue=0, description=""}
  User.RibbonIDOffset {type=NiagaraInt32}
  User.RightDot {type=NiagaraFloat, defaultValue=0, description=""}
  User.SkeletalMesh {
   type=NiagaraDataInterfaceSkeletalMesh
   defaultValue {
    struct=NiagaraExt_VariableValue_DataInterface
    value {dataInterface=&self:NiagaraDataInterfaceSkeletalMesh_0}
   }
  }
  User.Variants {type=NiagaraFloat, defaultValue=0, description=""}
 ]
}
