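# Niagara system for the Silencer pickaxe's swing trail, drawn by a single short-lived ribbon emitter.
# AnimTrail modules spawn and update the ribbon from bone or socket positions, while a curve tapers its width over particle life.
# Color is a constant dark gray that fades out over particle life, multiplied by an emitter-age curve that fades alpha in.
# User.HitWeakPoint plus User.Variants feeds a dynamic material parameter.
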NiagaraSystem /BRCosmetics/Weapons/FORT_Melee/Pickaxe_Silencer/FX/NS_Pickaxe_Silencer_Trail
dependencies {
 usedDataInterfaces [&NiagaraDataInterfaceCurve, &NiagaraDataInterfaceColorCurve]
 usedDynamicInputs []
 usedModules [&SystemState]
}
emitters {AirTrail001 {
  emitterUpdateScript [
   &EmitterState(Life Cycle Mode=::NewEnumerator1, Loop Behavior=::NewEnumerator1, Loop Duration=0.15, UseLoopDelay=true)
   &FN_SpawnAnimTrail_v2(Spawn Interval=0)
  ]
  particleSpawnScript [
   /Game/Effects/Niagara/Modules/AnimTrail/FN_InitializeAnimTrail/FN_InitializeAnimTrailBoneIndecies(Second Bone or Socket Index=1)
   /Game/Effects/Niagara/Modules/AnimTrail/FN_InitializeAnimTrail/FN_InitializeAnimTrail(AnimTrailWidth=1, Color=<rgba 1 1 1 100>, Lifetime=0.19)
   /SetVariables() {isSetParameters=true}
  ]
  particleUpdateScript [
   &SolveForcesAndVelocity(Clamp Velocity=false, Limit Acceleration=false)
   &ScaleRibbonWidth(Ribbon Width Scale=&FloatFromCurve(FloatCurve=DataInterface() {propertyValues {Curve {keys [
         {arriveTangent=-1.2579, interpMode=RCIM_Cubic, time=0, value=1}
         {arriveTangent=-0.9664, interpMode=RCIM_Cubic, leaveTangent=-0.9664, tangentMode=RCTM_User, time=1, value=0.2}
        ]}}}))
   &FN_UpdateAnimTrail_v2()
   &DynamicMaterialParameters(
    Index 0 Param 1=&MakeFloatFromInt(Int=&Add_Integer(A=$User.HitWeakPoint, B=&MakeIntFromFloat(Float=$User.Variants)))
    Index 0 Param 3=1
    Index 0 Param 4=1
    Module.Param0WriteEnabled=true
    Module.Param1WriteEnabled=false
    Module.Param2WriteEnabled=false
    Module.Param3WriteEnabled=false
   )
   &Color(Color=&MultiplyLinearColor(
     Linear Color=&ColorFromCurve(ColorCurve=DataInterface() {propertyValues {
        AlphaCurve {keys [<at 0.1132 1 RCIM_Cubic>,<at 0.9241 0 RCIM_Cubic>,<at 1 0>]}
        BlueCurve {keys [<at 0.0114 0.0677>]}
        GreenCurve {keys [<at 0.0114 0.0677>]}
        RedCurve {keys [<at 0.0114 0.0677>]}
       }})
     Scale Factor=&Vector4FromCurve(
      CurveIndex=$Emitter.Age
      Vector4Curve=DataInterface() {propertyValues {
        AlphaCurve {keys [<at 0.05 0>,<at 0.1 1>]}
        BlueCurve {keys [<at 0 1>,<at 1 1>]}
        GreenCurve {keys [<at 0 1>,<at 1 1>]}
        RedCurve {keys [<at 0 1>,<at 1 1>]}
       }}
     )
    ))
  ]
  rendererClasses [&NiagaraRibbonRendererProperties]
  renderers [{rendererClass=&NiagaraRibbonRendererProperties}]
 }}
summary {
 systemName=NS_Pickaxe_Silencer_Trail
 emitters [{emitterName=AirTrail001, rendererClasses [&NiagaraRibbonRendererProperties]}]
 userVariables [
  User.Color A {type=LinearColor, description="", defaultValue {struct=LinearColor, value=<rgba 1 1 1 1>}}
  User.HitWeakPoint {type=NiagaraInt32}
  User.LifetimeCustom {type=NiagaraFloat, defaultValue=0, description=""}
  User.Primary Color {type=LinearColor, description="", defaultValue {struct=LinearColor, value=<rgba 1 1 1 1>}}
  User.RibbonIDOffset {type=NiagaraInt32}
  User.SkeletalMesh {
   type=NiagaraDataInterfaceSkeletalMesh
   defaultValue {
    struct=NiagaraExt_VariableValue_DataInterface
    value {dataInterface=&self:NiagaraDataInterfaceSkeletalMesh_0}
   }
  }
  User.Variants {type=NiagaraFloat, defaultValue=0, description=""}
 ]
}
